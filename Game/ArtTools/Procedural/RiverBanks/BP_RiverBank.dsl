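; Procedural river bank actor that reshapes and paints the landscape along one side of a river and dresses it with meshes.
; Construction script finds the river water body by trace and copies its spline, offset toward the actor's side.
; Temporary spline meshes captured to alpha and depth targets, plus the rendered heightmap, feed noise materials drawn into landscape height and weight patches.
; Seeded streams place bank meshes along the offset spline and trace-scattered rock instances with secondary clusters.
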(event EventBeginPlay)

(event Collision|EventActorBeginOverlap
  (params (OtherActor :type Actor%20Object%20Reference)))

(event EventTick
  (params (DeltaSeconds :type Float%20%28single-precision%29)))

(event Custom|1%20ResetTerrain
  (LandscapePatch|SetIsEnabled (Variables|Default|GetLandscapeTexturePatch) false))

(event Custom|2%20Deform%20Landscape
  (LandscapePatch|SetIsEnabled (Variables|Default|GetLandscapeTexturePatch) true)
  (LandscapePatch|LandscapeEditMasterFunction))

(event Custom|UpdateHeight
  (LandscapePatch|SetIsEnabled (Variables|Default|GetLandscapeTexturePatch) true)
  (LandscapePatch|LandscapeEditMasterFunction))

(fn Add_SplineMesh_Components_ForBlend
  (params (Evaluate%20WPO :type Boolean) (Mesh :type Static%20Mesh%20Object%20Reference) (HeightOffset :type Float%20%28single-precision%29) (AlphaFalloff :type Float%20%28single-precision%29) (BlendDistance :type Float%20%28single-precision%29) (BaseWidthAdd :type Float%20%28double-precision%29) (BaseWidth :type Float%20%28single-precision%29) (Spline%20Source :type Spline%20Component%20Object%20Reference))
  (bind _localsplinecomponentlist (LocalVariables|GetLocalSplineComponentlist))
  (Utilities|FlowControl|Sequence
    (:then_0
      (for _index (range (Spline|GetNumberOfSplinePoints Spline%20Source))
        (LocalVariables|SetCurrentPoint _index)
        (bind _returnvalue (AddComponent|NODE_AddSplineMeshComponent-1 :RelativeTransform (Math|Transform|MakeTransform "0, 0, 0")))
        (Rendering|LOD|SetEvaluateWorldPositionOffset _returnvalue Evaluate%20WPO :!class /Script/Engine.StaticMeshComponent)
        (Rendering|SetBoundsScale _returnvalue 3.0)
        (bind _output_get (LocalVariables|SetCurrentSplineComponent _returnvalue))
        (Utilities|Array|Add _localsplinecomponentlist _output_get)
        (Components|StaticMesh|SetStaticMesh (LocalVariables|GetCurrentSplineComponent) Mesh)
        (Utilities|FlowControl|Sequence
          (:then_0
            (bind _mid (LandscapePatch|MakeMaterialForBlendMesh _returnvalue HeightOffset AlphaFalloff BlendDistance BaseWidthAdd BaseWidth)))
          (:then_1
            (bind (_scalestart _scaleend) (LandscapePatch|SetLocationfromSpline (LocalVariables|GetCurrentSplineComponent) (LocalVariables|GetCurrentPoint) (Variables|GetSpline%20Source)))))))
    (:then_1
      (return _localsplinecomponentlist))))

(fn AddSceneCapture
  (params (Spline :type Spline%20Component%20Object%20Reference) (LandscapePatch :type Landscape%20Texture%20Patch%20Object%20Reference) (Format :type ETextureRenderTargetFormat%20Enum :default "RTF_R8") (Capture%20Source :type ESceneCaptureSource%20Enum :default "SCS_SceneColorHDR"))
  (bind (_x _y) (Math|Vector2D|BreakVector2D (LandscapePatch|GetInitResolutionfromLandscape (Variables|GetLandscapePatch) 1.0)))
  (bind (_x_max _x_min _y_max _y_min _z_min _z_max) (LandscapePatch|GetSplineMeshBounds Spline "World"))
  (bind _returnvalue (AddComponent|NODE_AddSceneCaptureComponent2D-0 :RelativeTransform (Math|Transform|MakeTransform (Math|Vector|MakeVector 0.0 0.0 (+ (- _z_max (.z (Transformation|GetActorLocation))) 5e+02)) "-90.000000,-90.000000,0.000000")))
  (bind _output_get (LocalVariables|SetLocalSceneCapture _returnvalue))
  (Variables|/Script/Engine.SceneCaptureComponent|SetShowOnlyActors (Utilities|Array|MakeArray self) _output_get)
  (Variables|/Script/Engine.SceneCaptureComponent|SetCaptureSource Capture%20Source _returnvalue)
  (Variables|/Script/Engine.SceneCaptureComponent2D|SetOrthoWidth (Math|Vector2D|BreakVector2D (LandscapePatch|GetUnscaledCoverage (Variables|GetLandscapePatch))) _output_get)
  (bind _returnvalue_2 (Rendering|CreateRenderTarget2D (Math|Float|Truncate _x) (Math|Float|Truncate _y) Format))
  (Variables|/Script/Engine.SceneCaptureComponent2D|SetTextureTarget _returnvalue_2 (LocalVariables|GetLocalSceneCapture))
  (return _returnvalue_2 (LocalVariables|GetLocalSceneCapture)))

(fn AssignRiver
  (Utilities|IsValid (Variables|Default|GetRiverWaterBodyActor)
    (:Is%20Not%20Valid
      (bind _outhit (Collision|LineTraceByChannel (+ (Transformation|GetActorLocation) "0,0,500000.000000") (+ (Transformation|GetActorLocation) "0,0,-5000000.000000") "TraceTypeQuery12"))
      (bind _asfort_water_body_bp (Utilities|Casting|CastToFortWaterBodyBP (.HitActor (Collision|BreakHitResult _outhit))))
      (if (Utilities|Enum|Equal(Enum) (Water|GetWaterBodyType _asfort_water_body_bp) "River")
        (Variables|Default|SetRiverWaterBodyActor _asfort_water_body_bp)))))

(fn CaputreLandscapeHeight
  (params (Center%20WorldSpace :type Vector) (LandscapeTexturePatch1 :type Landscape%20Texture%20Patch%20Object%20Reference) (Size :type Float%20%28double-precision%29))
  (bind (_x _y) (Math|Vector2D|BreakVector2D (LandscapePatch|GetInitResolutionfromLandscape LandscapeTexturePatch1 1.0)))
  (bind _outactors (Actor|GetAllActorsOfClass "/Script/Landscape.Landscape"))
  (bind _returnvalue (Rendering|CreateRenderTarget2D (Math|Float|Truncate _x) (Math|Float|Truncate _y) "RTF_RGBA8"))
  (Landscape|Runtime|RenderHeightmap (Utilities|Array|Get(acopy) _outactors) (Math|Transform|MakeTransform Center%20WorldSpace) (Math|Box2D|MakeBox2D (Math|Vector2D|MakeVector2D (* (/ (Variables|GetSize) 2.0) -1.0) (* (/ (Variables|GetSize) 2.0) -1.0)) (Math|Vector2D|MakeVector2D (/ (Variables|GetSize) 2.0) (/ (Variables|GetSize) 2.0))) _returnvalue)
  (return _returnvalue))

(fn CopySplineSection
  (bind _returnvalue (Transformation|GetActorLocation))
  (bind _waterbodyspline (LocalVariables|GetWaterBodySpline))
  (bind _returnvalue_2 (Spline|FindRightVectorClosesttoWorldLocation _waterbodyspline _returnvalue "World"))
  (bind _returnvalue_3 (Math|Float|MakeLiteralFloat (Variables|Default|GetWidthReplaceWithGet)))
  (bind _waterbodyspline_2 (LocalVariables|GetWaterBodySpline))
  (bind _returnvalue_4 (Spline|FindLocationClosesttoWorldLocation _waterbodyspline_2 _returnvalue "World"))
  (bind _returnvalue_5 (Math|Vector|GetUnitDirection(Vector) _returnvalue _returnvalue_4))
  (bind _returnvalue_6 (Math|Vector|DotProduct _returnvalue_5 _returnvalue_2))
  (bind _returnvalue_7 (Math|Float|Sign(Float) _returnvalue_6))
  (bind _waterbodyspline_3 (LocalVariables|GetWaterBodySpline))
  (bind _returnvalue_8 (Spline|FindInputKeyClosesttoWorldLocation _waterbodyspline_3 _returnvalue))
  (bind _returnvalue_9 (Math|Float|Truncate _returnvalue_8))
  (bind _returnvalue_10 (+ _returnvalue_9 1))
  (bind _returnvalue_11 (* _returnvalue_3 _returnvalue_7))
  (bind _waterbodyspline_4 (LocalVariables|GetWaterBodySpline))
  (Utilities|IsValid (Variables|Default|GetRiverWaterBodyActor)
    (:Is%20Valid
      (LocalVariables|SetWaterBodySpline (Variables|/Script/Water.WaterBody|GetSplineComp (Variables|Default|GetRiverWaterBodyActor)))
      (bind _returnvalue_12 (AddComponent|NODE_AddSplineComponent-0 :RelativeTransform (Math|Transform|MakeTransform "0, 0, 0")))
      (bind _output_get (LocalVariables|SetNewSpline _returnvalue_12))
      (Spline|RemoveSplinePoint _returnvalue_12 1 true)
      (Spline|RemoveSplinePoint _returnvalue_12 0 true)
      (Utilities|FlowControl|Sequence
        (:then_0
          (for _index (range (Spline|GetNumberOfSplinePoints (LocalVariables|GetWaterBodySpline)))
            (Spline|AddSplinePoint (LocalVariables|GetNewSpline) (+ (.location (Spline|GetTransformatSplinePoint _waterbodyspline_4 _index "World" false)) (* (Math|Vector|NegateVector (Math|Vector|Normalize (Spline|GetRightVectoratSplinePoint (LocalVariables|GetWaterBodySpline) _index "World"))) _returnvalue_11)) "World" true)
            (Spline|SetTangentatSplinePoint (LocalVariables|GetNewSpline) 0 (Math|Vector|GetForwardVector (.rotation (Spline|GetTransformatSplinePoint _waterbodyspline_4 _index "World" false))) "World" true)))
        (:then_1
          (Variables|Default|SetIsLeftorRight (> _returnvalue_7 0))
          (bind _output_get_2 (LocalVariables|SetBaseLocation (+ (- _returnvalue_4 _returnvalue) (* (Math|Vector|Normalize (Spline|FindRightVectorClosesttoWorldLocation (LocalVariables|GetWaterBodySpline) _returnvalue "Local")) _returnvalue_11))))
          (Spline|AddSplinePointatIndex (LocalVariables|GetNewSpline) _output_get_2 _returnvalue_10 "Local" true)
          (LocalVariables|SetBaseDistance (Spline|GetDistanceAlongSplineatSplinePoint (LocalVariables|GetNewSpline) _returnvalue_10))
          (Spline|RemoveSplinePoint (LocalVariables|GetNewSpline) _returnvalue_10 true))
        (:then_2
          (return (LocalVariables|GetNewSpline) (LocalVariables|GetWaterBodySpline) (LocalVariables|GetBaseLocation) (LocalVariables|GetBaseDistance)))))))

(fn DebugRenderTarget
  (params (Debug :type Boolean) (Texture :type Texture%20Object%20Reference) (Range :type Float%20%28single-precision%29 :default 1.0))
  (if Debug
    (bind _returnvalue (AddComponent|NODE_AddStaticMeshComponent-1 :RelativeTransform (Math|Transform|MakeTransform "0, 0,1000.000000" "0, 0, 0" "3.000000,3.000000,3.000000")))
    (bind _returnvalue_2 (Rendering|Material|CreateDynamicMaterialInstance _returnvalue 0 "/Game/ArtTools/Procedural/RiverBanks/M_DebugRenderTarget.M_DebugRenderTarget" "None" :!class /Script/Engine.PrimitiveComponent))
    (Rendering|Material|SetTextureParameterValue _returnvalue_2 "Texture" Texture)
    (Rendering|Material|SetScalarParameterValue _returnvalue_2 "Range" Range :!class /Script/Engine.MaterialInstanceDynamic)))

(fn GetSplineMeshBounds
  (params (Spline :type Spline%20Component%20Object%20Reference) (CoordinateSpace :type ESplineCoordinateSpace%20Enum :default "Local"))
  (Utilities|FlowControl|Sequence
    (:then_0
      (for _index (range (Spline|GetNumberOfSplinePoints Spline))
        (LocalVariables|SetX_Max (Math|Float|Max(Float) (.x (Spline|GetLocationatSplinePoint Spline _index CoordinateSpace)) (LocalVariables|GetX_Max)))
        (LocalVariables|SetX_Min (Math|Float|Min(Float) (LocalVariables|GetX_Min) (.x (Spline|GetLocationatSplinePoint Spline _index CoordinateSpace))))
        (LocalVariables|SetY_Max (Math|Float|Max(Float) (LocalVariables|GetY_Max) (.y (Spline|GetLocationatSplinePoint Spline _index CoordinateSpace))))
        (LocalVariables|SetY_Min (Math|Float|Min(Float) (LocalVariables|GetY_Min) (.y (Spline|GetLocationatSplinePoint Spline _index CoordinateSpace))))
        (LocalVariables|SetZ_Max (Math|Float|Max(Float) (.z (Spline|GetLocationatSplinePoint Spline _index CoordinateSpace)) (LocalVariables|GetZ_Max)))
        (LocalVariables|SetZ_Min (Math|Float|Min(Float) (.z (Spline|GetLocationatSplinePoint Spline _index CoordinateSpace)) (LocalVariables|GetZ_Min)))
        (LocalVariables|SetScale_Y_Max (Math|Float|Max(Float) (LocalVariables|GetScale_Y_Max) (.y (Spline|GetScaleatSplinePoint Spline _index))))))
    (:then_1
      (return (LocalVariables|GetX_Max) (LocalVariables|GetX_Min) (LocalVariables|GetY_Max) (LocalVariables|GetY_Min) (LocalVariables|GetZ_Min) (LocalVariables|GetZ_Max) (- (LocalVariables|GetX_Max) (LocalVariables|GetX_Min)) (- (LocalVariables|GetY_Max) (LocalVariables|GetY_Min)) (- (LocalVariables|GetZ_Max) (LocalVariables|GetZ_Min)) (Math|Vector|MakeVector (- (LocalVariables|GetX_Max) (/ (- (LocalVariables|GetX_Max) (LocalVariables|GetX_Min)) 2.0)) (- (LocalVariables|GetY_Max) (/ (- (LocalVariables|GetY_Max) (LocalVariables|GetY_Min)) 2.0)) (- (LocalVariables|GetZ_Max) (/ (- (LocalVariables|GetZ_Max) (LocalVariables|GetZ_Min)) 2.0))) (+ (Math|Vector|MakeVector (- (LocalVariables|GetX_Max) (/ (- (LocalVariables|GetX_Max) (LocalVariables|GetX_Min)) 2.0)) (- (LocalVariables|GetY_Max) (/ (- (LocalVariables|GetY_Max) (LocalVariables|GetY_Min)) 2.0)) (- (LocalVariables|GetZ_Max) (/ (- (LocalVariables|GetZ_Max) (LocalVariables|GetZ_Min)) 2.0))) (Transformation|GetActorLocation)) (LocalVariables|GetScale_Y_Max)))))

(fn LandscapeEdit_MasterFunction
  (bind _landscapetexturepatch (Variables|Default|GetLandscapeTexturePatch))
  (LandscapePatch|SetIsEnabled (Variables|Default|GetLandscapeTexturePatch) true)
  (LandscapePatch|SetupLandscapePatch (Variables|Default|GetLandscapeTexturePatch) (Math|Vector2D|MakeVector2D (Variables|Default|GetLength) (Variables|Default|GetLength)) true)
  (Utilities|IsValid (Variables|Default|GetRiverWaterBodyActor)
    (:Is%20Valid
      (Utilities|FlowControl|Sequence
        (:then_0
          (bind _rendertarget (LandscapePatch|CaputreLandscapeHeight (Transformation|GetWorldLocation (Variables|Default|GetLandscapeTexturePatch)) (Variables|Default|GetLandscapeTexturePatch) (Variables|Default|GetLength)))
          (bind _output_get (LocalVariables|SetRenderTarget_LandscapeHeight _rendertarget))
          (CallFunction|DebugRenderTarget false _output_get))
        (:then_1
          (CallFunction|SetVisiblityAllComponents false)
          (bind _allsplinecomponents (LandscapePatch|AddSplineMeshComponentsforBlend true "/Game/ArtTools/Procedural/LandscapeBlend/SplineMeshBlend/SplinemeshLandscapeBlend.SplinemeshLandscapeBlend" 0.0 3.0 5e+02 1.0 (Variables|Default|GetWidthReplaceWithGet) (Variables|/Script/Water.WaterBody|GetSplineComp (Variables|Default|GetRiverWaterBodyActor))))
          (Utilities|FlowControl|Sequence
            (:then_0
              (bind (_rendertarget_2 _scenecapture) (CallFunction|AddSceneCapture (Variables|/Script/Water.WaterBody|GetSplineComp (Variables|Default|GetRiverWaterBodyActor)) (Variables|Default|GetLandscapeTexturePatch) "RTF_R8" "SCS_SceneColorHDRNoAlpha")) ; Alpha
              (LocalVariables|SetRenderTarget_Alpha _rendertarget_2)
              (Rendering|SceneCapture|CaptureScene _scenecapture :!class /Script/Engine.SceneCaptureComponent2D)
              (CallFunction|DebugRenderTarget false _rendertarget_2))
            (:then_1
              (bind (_rendertarget_3 _scenecapture_2) (CallFunction|AddSceneCapture (Variables|/Script/Water.WaterBody|GetSplineComp (Variables|Default|GetRiverWaterBodyActor)) (Variables|Default|GetLandscapeTexturePatch) "RTF_R32f" "SCS_SceneDepth")) ; SplineHeight
              (LocalVariables|SetRenderTarget_SplineHeight _rendertarget_3)
              (bind _output_get_2 (LocalVariables|SetSceneCapture_SplineHeight _scenecapture_2))
              (Rendering|SceneCapture|CaptureScene _output_get_2 :!class /Script/Engine.SceneCaptureComponent2D)
              (CallFunction|DebugRenderTarget false (LocalVariables|GetRenderTarget_SplineHeight) 5e+03))
            (:then_2
              (if (not false)
                (for _array_element _allsplinecomponents
                  (Components|DestroyComponent _array_element))
                (Utilities|Array|Clear _allsplinecomponents)))))
        (:then_2
          (bind _returnvalue (Rendering|Material|CreateDynamicMaterialInstance "/Game/ArtTools/Procedural/RiverBanks/M_PerlinNoise_height.M_PerlinNoise_height" "None" "None" :!class /Script/Engine.KismetMaterialLibrary))
          (bind _output_get_3 (LocalVariables|SetMID_NewHeightMat _returnvalue))
          (Rendering|Material|SetTextureParameterValue _output_get_3 "Alpha" (LocalVariables|GetRenderTarget_Alpha))
          (Rendering|Material|SetTextureParameterValue (LocalVariables|GetMID_NewHeightMat) "LandscapeHeight" (LocalVariables|GetRenderTarget_LandscapeHeight))
          (Rendering|Material|SetTextureParameterValue (LocalVariables|GetMID_NewHeightMat) "SplineHeight" (LocalVariables|GetRenderTarget_SplineHeight))
          (Rendering|Material|SetScalarParameterValue (LocalVariables|GetMID_NewHeightMat) "Z_PositionCamera" (.z (Transformation|GetWorldLocation (LocalVariables|GetSceneCapture_SplineHeight))) :!class /Script/Engine.MaterialInstanceDynamic)
          (Rendering|Material|SetScalarParameterValue (LocalVariables|GetMID_NewHeightMat) "Size" (Variables|LandscapeDeformation|GetHeightNoise_Size) :!class /Script/Engine.MaterialInstanceDynamic)
          (Rendering|Material|SetScalarParameterValue (LocalVariables|GetMID_NewHeightMat) "HeightAbove" (Variables|LandscapeDeformation|GetHeight_AboveWaterline) :!class /Script/Engine.MaterialInstanceDynamic)
          (Rendering|Material|SetScalarParameterValue (LocalVariables|GetMID_NewHeightMat) "HeightBelow" (Variables|LandscapeDeformation|GetHeight%20Below%20Waterline) :!class /Script/Engine.MaterialInstanceDynamic)
          (Rendering|Material|SetScalarParameterValue (LocalVariables|GetMID_NewHeightMat) "MaxHeightAboveWater" (Variables|LandscapeDeformation|GetMax%20Height%20Above%20Water) :!class /Script/Engine.MaterialInstanceDynamic)
          (Rendering|Material|SetScalarParameterValue (LocalVariables|GetMID_NewHeightMat) "Random" (+ (.x (Transformation|GetActorLocation)) (.y (Transformation|GetActorLocation))) :!class /Script/Engine.MaterialInstanceDynamic)
          (Rendering|Material|SetScalarParameterValue (LocalVariables|GetMID_NewHeightMat) "WorldSize" (Variables|Default|GetLength) :!class /Script/Engine.MaterialInstanceDynamic)
          (Rendering|Material|SetScalarParameterValue (LocalVariables|GetMID_NewHeightMat) "EdgeFade" (Variables|LandscapeDeformation|GetEndFadeOut) :!class /Script/Engine.MaterialInstanceDynamic)
          ;[1 1 1] Get landscapeheight
          (bind _outactors (Actor|GetAllActorsOfClass "/Script/Landscape.LandscapeProxy"))
          (for _array_element_2 _outactors
            (Rendering|Material|SetScalarParameterValue (LocalVariables|GetMID_NewHeightMat) "Z_PositionTerrain" (.z (Transformation|GetActorLocation _array_element_2)) :!class /Script/Engine.MaterialInstanceDynamic)))
        (:then_3
          (bind _returnvalue_2 (Rendering|Material|CreateDynamicMaterialInstance "/Game/ArtTools/Procedural/RiverBanks/M_RiverBank_WeightLayerPaint.M_RiverBank_WeightLayerPaint" "None" "None" :!class /Script/Engine.KismetMaterialLibrary))
          (bind _output_get_4 (LocalVariables|SetMID_WeightPatch _returnvalue_2))
          (Rendering|Material|SetTextureParameterValue _output_get_4 "Alpha" (LocalVariables|GetRenderTarget_Alpha))
          (Rendering|Material|SetScalarParameterValue (LocalVariables|GetMID_WeightPatch) "Size" (Variables|LandscapeDeformation|GetHeightNoise_Size) :!class /Script/Engine.MaterialInstanceDynamic)
          (Rendering|Material|SetScalarParameterValue (LocalVariables|GetMID_WeightPatch) "Random" (+ (.x (Transformation|GetActorLocation)) (.y (Transformation|GetActorLocation))) :!class /Script/Engine.MaterialInstanceDynamic)
          (Rendering|Material|SetScalarParameterValue (LocalVariables|GetMID_WeightPatch) "WorldSize" (Variables|Default|GetLength) :!class /Script/Engine.MaterialInstanceDynamic)
          (Rendering|Material|SetScalarParameterValue (LocalVariables|GetMID_WeightPatch) "EdgeFade" (Variables|LandscapeDeformation|GetEndFadeOut) :!class /Script/Engine.MaterialInstanceDynamic))
        (:then_4
          (CallFunction|SetVisiblityAllComponents true)
          (bind _returnvalue_3 (LandscapePatch|GetHeightRenderTarget _landscapetexturepatch true))
          (Rendering|DrawMaterialtoRenderTarget _returnvalue_3 (LocalVariables|GetMID_NewHeightMat))
          (LandscapePatch|RequestLandscapeUpdate _landscapetexturepatch false true))
        (:then_5
          (LandscapePatch|CreateWeightPatch (Variables|Default|GetLandscapeTexturePatch) (Variables|LandscapeDeformation|GetPaintLayer%20Name) "TextureBackedRenderTarget" "SourceTextureChannel")
          (bind _returnvalue_4 (LandscapePatch|GetWeightPatchRenderTarget (Variables|Default|GetLandscapeTexturePatch) (Variables|LandscapeDeformation|GetPaintLayer%20Name) true))
          (Rendering|DrawMaterialtoRenderTarget _returnvalue_4 (LocalVariables|GetMID_WeightPatch)))))))

(fn MakeMaterialforBlendMesh
  (params (Component :type Primitive%20Component%20Object%20Reference) (HeightOffset :type Float%20%28single-precision%29) (AlphaFalloff :type Float%20%28single-precision%29) (BlendDistance :type Float%20%28single-precision%29) (BaseWidthAdd :type Float%20%28double-precision%29) (BaseWidth :type Float%20%28single-precision%29))
  (bind _blendmaterial (LocalVariables|GetBlendMaterial))
  (bind _blendmaterial_2 (LocalVariables|GetBlendMaterial))
  (bind _returnvalue (Rendering|Material|CreateDynamicMaterialInstance "/Game/ArtTools/Procedural/LandscapeBlend/SplineMeshBlend/SplineBlendMesh.SplineBlendMesh" "None" "None" :!class /Script/Engine.KismetMaterialLibrary))
  (LocalVariables|SetBlendMaterial _returnvalue)
  (Rendering|Material|SetScalarParameterValue _returnvalue "BaseWidth" (+ BaseWidth BaseWidthAdd) :!class /Script/Engine.MaterialInstanceDynamic)
  (Rendering|Material|SetScalarParameterValue _blendmaterial_2 "HeightOffset" HeightOffset :!class /Script/Engine.MaterialInstanceDynamic)
  (Rendering|Material|SetScalarParameterValue _blendmaterial_2 "AlphaFalloff" AlphaFalloff :!class /Script/Engine.MaterialInstanceDynamic)
  (Rendering|Material|SetScalarParameterValue _blendmaterial_2 "BlendDistance" BlendDistance :!class /Script/Engine.MaterialInstanceDynamic)
  (Rendering|Material|SetScalarParameterValue _returnvalue "CenterBlend" (Variables|LandscapeDeformation|GetCenterFadeDistance) :!class /Script/Engine.MaterialInstanceDynamic)
  (Rendering|Material|SetScalarParameterValue (LocalVariables|GetBlendMaterial) "LeftSideMask" (select (Variables|Default|GetIsLeftorRight) 0.0 1.0) :!class /Script/Engine.MaterialInstanceDynamic)
  (Rendering|Material|SetScalarParameterValue (LocalVariables|GetBlendMaterial) "RightSideMask" (select (Variables|Default|GetIsLeftorRight) 1.0 0.0) :!class /Script/Engine.MaterialInstanceDynamic)
  (Rendering|Material|SetMaterial Component 0 _blendmaterial)
  (return _blendmaterial))

(fn SetLocationFromSpline
  (params (SplineComponant :type Spline%20Mesh%20Component%20Object%20Reference) (CurrentPoint :type Integer) (SplineSource :type Spline%20Component%20Object%20Reference))
  (bind _returnvalue (+ CurrentPoint 1))
  (bind (_location _tangent) (Spline|GetLocationandTangentatSplinePoint SplineSource _returnvalue "World"))
  (bind (_location_2 _tangent_2) (Spline|GetLocationandTangentatSplinePoint SplineSource CurrentPoint "World"))
  (bind _splinecomponant (Variables|GetSplineComponant))
  (Utilities|FlowControl|Sequence
    (:then_0
      (LocalVariables|SetCurrenPointLocation _location_2)
      (LocalVariables|SetCurrentTangent _tangent_2)
      (LocalVariables|SetNextPointLocation _location)
      (LocalVariables|SetNextTangent _tangent)
      (SplineMesh|SetStartandEnd SplineComponant (- (LocalVariables|GetCurrenPointLocation) (Transformation|GetActorLocation)) (LocalVariables|GetCurrentTangent) (- (LocalVariables|GetNextPointLocation) (Transformation|GetActorLocation)) (LocalVariables|GetNextTangent) true :!class /Script/Engine.SplineMeshComponent)
      (SplineMesh|SetStartRollRadians _splinecomponant (Math|Trig|DegreesToRadians (Spline|GetRollatSplinePoint SplineSource CurrentPoint "Local")) true)
      (SplineMesh|SetEndRollRadians _splinecomponant (Math|Trig|DegreesToRadians (Spline|GetRollatSplinePoint SplineSource _returnvalue "Local")) true))
    (:then_1
      (return (.y (Spline|GetScaleatSplinePoint (Variables|GetSplineSource) CurrentPoint)) (.y (Spline|GetScaleatSplinePoint (Variables|GetSplineSource) _returnvalue))))))

(fn SetupLandscapePatch
  (params (Landscape%20Texture%20Patch :type Landscape%20Texture%20Patch%20Object%20Reference) (Coverage :type Vector%202D%20Structure) (Use%20Alpha :type Boolean))
  (LandscapePatch|SetUnscaledCoverage Landscape%20Texture%20Patch Coverage)
  (LandscapePatch|SetResolution Landscape%20Texture%20Patch (LandscapePatch|GetInitResolutionfromLandscape Landscape%20Texture%20Patch 1.0))
  (if (Variables|GetUse%20Alpha)
    (LandscapePatch|SetHeightRenderTargetFormat Landscape%20Texture%20Patch "RTF_RGBA32f")
    (Transformation|SetRelativeLocation (Variables|Default|GetLandscapeTexturePatch) "0, 0, 0" false false)))

(fn SetVisiblityAllComponents
  (params (Visibility :type Boolean))
  (Rendering|SetVisibility (Variables|Components|GetInstancedStaticMesh) Visibility false)
  (for _array_element (Variables|Default|GetAllStaticMeshComponents)
    (Rendering|SetVisibility _array_element Visibility false)))

(fn SpawnLargeBankMeshes
  (params (SplineEdge :type Spline%20Component%20Object%20Reference) (SplineCenter :type Spline%20Component%20Object%20Reference) (BaseLocation :type Vector) (Distance%20on%20Spline :type Float%20%28double-precision%29))
  (bind _currentstaticmesh (LocalVariables|GetCurrentStaticMesh))
  (bind _returnvalue (StaticMesh|GetBounds _currentstaticmesh))
  (bind (_origin _boxextent) (Math|BoxSphereBounds|BreakBoxSphereBounds _returnvalue))
  (bind _x (.x _boxextent))
  (bind _currentdistance (LocalVariables|GetCurrentDistance))
  (bind _length (Variables|Default|GetLength))
  (bind _returnvalue_2 (- _length _currentdistance))
  (if (> (Utilities|Array|Length (Variables|Default|GetStaticMeshList)) 0)
    (bind _index (Utilities|FlowControl|ForLoopwithBreak 0 1000)
      (:LoopBody
        (Math|Random|SetRandomStreamSeed (Variables|Default|GetStream) (+ _index (Variables|Default|GetInitial%20Seed)))
        (LocalVariables|SetCurrentStaticMesh (Utilities|Array|Get(acopy) (Variables|Default|GetStaticMeshList) (Math|Random|RandomIntegerinRangefromStream (Variables|Default|GetStream) 0 (Utilities|Array|LastIndex (Variables|Default|GetStaticMeshList)))))
        (LocalVariables|SetOffsetPivot (* (+ _x (.x (Utilities|Struct|BreakBox (StaticMesh|GetBoundingBox _currentstaticmesh)))) -1.0))
        (LocalVariables|SetHalfAssetWidth _x)
        (if (< _returnvalue_2 (* (+ (Variables|Default|GetMesh%20Gap%20Distance) (* (LocalVariables|GetHalfAssetWidth) 2.0)) 0.8))
          (if (< _returnvalue_2 1e+02)
            _))
        (bind _returnvalue_3 (AddComponent|NODE_AddStaticMeshComponent-0 :RelativeTransform (Math|Transform|MakeTransform (+ (- (+ (Spline|GetLocationatDistanceAlongSpline SplineEdge (Utilities|Operators|Add (LocalVariables|GetHalfAssetWidth) (- Distance%20on%20Spline (/ (Variables|Default|GetLength) 2.0)) (LocalVariables|GetCurrentDistance) (LocalVariables|GetOffsetPivot) (Variables|Default|GetMesh%20Gap%20Distance)) "World") (* (Math|Vector|GetUnitDirection(Vector) (Spline|GetLocationatDistanceAlongSpline SplineEdge (Utilities|Operators|Add (LocalVariables|GetHalfAssetWidth) (- Distance%20on%20Spline (/ (Variables|Default|GetLength) 2.0)) (LocalVariables|GetCurrentDistance) (LocalVariables|GetOffsetPivot) (Variables|Default|GetMesh%20Gap%20Distance)) "World") (Spline|GetLocationatDistanceAlongSpline SplineCenter (Utilities|Operators|Add (LocalVariables|GetHalfAssetWidth) (- Distance%20on%20Spline (/ (Variables|Default|GetLength) 2.0)) (LocalVariables|GetCurrentDistance) (LocalVariables|GetOffsetPivot) (Variables|Default|GetMesh%20Gap%20Distance)) "World")) (* (Math|Vector|Distance(Vector) (* (Transformation|GetActorLocation) "1.000000,1.000000,0") (* (Spline|FindLocationClosesttoWorldLocation SplineEdge (Transformation|GetActorLocation) "World") "1.000000,1.000000,0")) (Math|Float|Sign(Float) (Math|Vector|DotProduct (* (Math|Vector|GetUnitDirection(Vector) (Spline|GetLocationatDistanceAlongSpline SplineEdge (Utilities|Operators|Add (LocalVariables|GetHalfAssetWidth) (- Distance%20on%20Spline (/ (Variables|Default|GetLength) 2.0)) (LocalVariables|GetCurrentDistance) (LocalVariables|GetOffsetPivot) (Variables|Default|GetMesh%20Gap%20Distance)) "World") (Spline|GetLocationatDistanceAlongSpline SplineCenter (Utilities|Operators|Add (LocalVariables|GetHalfAssetWidth) (- Distance%20on%20Spline (/ (Variables|Default|GetLength) 2.0)) (LocalVariables|GetCurrentDistance) (LocalVariables|GetOffsetPivot) (Variables|Default|GetMesh%20Gap%20Distance)) "World")) "1.000000,1.000000,0") (Math|Vector|GetUnitDirection(Vector) (* (Spline|FindLocationClosesttoWorldLocation SplineEdge (Transformation|GetActorLocation) "World") "1.000000,1.000000,0") (* (Transformation|GetActorLocation) "1.000000,1.000000,0"))))))) (Transformation|GetActorLocation)) (Math|Vector|MakeVector 0.0 0.0 (- (.z (Transformation|GetActorLocation)) (.z (Spline|FindLocationClosesttoWorldLocation SplineEdge (Transformation|GetActorLocation) "World"))))) (Math|Rotator|MakeRotfromX (Spline|GetDirectionatDistanceAlongSpline SplineEdge (Utilities|Operators|Add (LocalVariables|GetHalfAssetWidth) (- Distance%20on%20Spline (/ (Variables|Default|GetLength) 2.0)) (LocalVariables|GetCurrentDistance) (LocalVariables|GetOffsetPivot) (Variables|Default|GetMesh%20Gap%20Distance)) "World")))))
        (Components|StaticMesh|SetStaticMesh _returnvalue_3 (LocalVariables|GetCurrentStaticMesh))
        (LocalVariables|SetCurrentDistance (Utilities|Operators|Add (LocalVariables|GetCurrentDistance) (* (LocalVariables|GetHalfAssetWidth) 2.0) (Variables|Default|GetMesh%20Gap%20Distance)))
        (Utilities|Array|Add (Variables|Default|GetAllStaticMeshComponents) _returnvalue_3)))))

(fn SpawnRocksScatter
  (params (SplineEdge :type Spline%20Component%20Object%20Reference) (SplineCenter :type Spline%20Component%20Object%20Reference) (BaseLocation :type Vector) (Distance%20on%20Spline :type Float%20%28double-precision%29))
  (bind _initial_seed (Variables|Default|GetInitial%20Seed))
  (for _index (range (+ (Variables|Default|GetNumber%20Of%20Scatter%20Meshes) 1))
    (Math|Random|SetRandomStreamSeed (Variables|Default|GetStream) (* _index _initial_seed))
    (Math|Random|SetRandomStreamSeed (Variables|Default|GetStream_2) (+ _index _initial_seed))
    (bind _outhits (Collision|MultiLineTraceForObjects (+ (Math|Random|RandomPointinBoundingBoxfromStream (Variables|Default|GetStream) (+ (Spline|GetLocationatDistanceAlongSpline SplineEdge (+ (- Distance%20on%20Spline (/ (Variables|Default|GetLength) 2.0)) (* (Variables|Default|GetLength) (Math|Float|Fraction (Math|Random|RandomFloatinRangefromStream (Variables|Default|GetStream_2) 0.0 5e+03)))) "World") (* (Math|Vector|GetUnitDirection(Vector) (Spline|GetLocationatDistanceAlongSpline SplineEdge (+ (- Distance%20on%20Spline (/ (Variables|Default|GetLength) 2.0)) (* (Variables|Default|GetLength) (Math|Float|Fraction (Math|Random|RandomFloatinRangefromStream (Variables|Default|GetStream_2) 0.0 5e+03)))) "World") (Spline|GetLocationatDistanceAlongSpline SplineCenter (+ (- Distance%20on%20Spline (/ (Variables|Default|GetLength) 2.0)) (* (Variables|Default|GetLength) (Math|Float|Fraction (Math|Random|RandomFloatinRangefromStream (Variables|Default|GetStream_2) 0.0 5e+03)))) "World")) (* (Variables|Default|GetSpawnEdgeWidth) (Math|Float|Fraction (Math|Random|RandomFloatinRangefromStream (Variables|Default|GetStream) 0.0 5e+03))))) "100.000000,100.000000, 0") "0,0,5000.000000") (- (Math|Random|RandomPointinBoundingBoxfromStream (Variables|Default|GetStream) (+ (Spline|GetLocationatDistanceAlongSpline SplineEdge (+ (- Distance%20on%20Spline (/ (Variables|Default|GetLength) 2.0)) (* (Variables|Default|GetLength) (Math|Float|Fraction (Math|Random|RandomFloatinRangefromStream (Variables|Default|GetStream_2) 0.0 5e+03)))) "World") (* (Math|Vector|GetUnitDirection(Vector) (Spline|GetLocationatDistanceAlongSpline SplineEdge (+ (- Distance%20on%20Spline (/ (Variables|Default|GetLength) 2.0)) (* (Variables|Default|GetLength) (Math|Float|Fraction (Math|Random|RandomFloatinRangefromStream (Variables|Default|GetStream_2) 0.0 5e+03)))) "World") (Spline|GetLocationatDistanceAlongSpline SplineCenter (+ (- Distance%20on%20Spline (/ (Variables|Default|GetLength) 2.0)) (* (Variables|Default|GetLength) (Math|Float|Fraction (Math|Random|RandomFloatinRangefromStream (Variables|Default|GetStream_2) 0.0 5e+03)))) "World")) (* (Variables|Default|GetSpawnEdgeWidth) (Math|Float|Fraction (Math|Random|RandomFloatinRangefromStream (Variables|Default|GetStream) 0.0 5e+03))))) "100.000000,100.000000, 0") "0,0,5000.000000") (Utilities|Array|MakeArray "ObjectTypeQuery1") true))
    (bind _array_element (Utilities|Array|ForEachLoopwithBreak _outhits)
      (:LoopBody
        (Utilities|Casting|CastToLandscapeProxy (.HitActor (Collision|BreakHitResult _array_element)))
        (Utilities|FlowControl|Sequence
          (:then_0
            _)
          (:then_1
            _))
        (Components|InstancedStaticMesh|AddInstance (Variables|Components|GetInstancedStaticMesh) (Math|Transform|MakeTransform (+ (- (.Location (Collision|BreakHitResult _array_element)) (Transformation|GetActorLocation)) (Math|Vector|MakeVector 0.0 0.0 (Variables|Default|GetZ_Offset))) (Math|Rotator|CombineRotators (Math|Rotator|MakeRotationfromAxes "1.000000, 0, 0" "0,1.000000, 0" (.Normal (Collision|BreakHitResult _array_element))) (Math|Random|RandomRotatorfromStream (Variables|Default|GetStream))) (Math|Vector|MakeVector (Math|Random|RandomFloatinRangefromStream (Variables|Default|GetStream) (Variables|Default|GetMin_Scale) (Variables|Default|GetMax_Scale)) (Math|Random|RandomFloatinRangefromStream (Variables|Default|GetStream) (Variables|Default|GetMin_Scale) (Variables|Default|GetMax_Scale)) (Math|Random|RandomFloatinRangefromStream (Variables|Default|GetStream) (Variables|Default|GetMin_Scale) (Variables|Default|GetMax_Scale)))) false)
        (for _ (range 1 (+ (Math|Random|RandomIntegerinRangefromStream (Variables|Default|GetStream) (Variables|Default|GetMinNumberofSecondaryScattermeshes) (Variables|Default|GetMaxNumberofSecondaryScattermeshes)) 1))
          (bind _outhits_2 (Collision|MultiLineTraceForObjects :Start (+ (+ (.Location (Collision|BreakHitResult _array_element)) (* (Math|Vector|Normalize (Math|Vector|GetForwardVector (Math|Random|RandomRotatorfromStream (Variables|Default|GetStream)))) (Math|Random|RandomFloatinRangefromStream (Variables|Default|GetStream) (Variables|Default|GetMinRadius) (Variables|Default|GetMax%20Radius)))) "0,0,50000.000000") :End (+ (+ (.Location (Collision|BreakHitResult _array_element)) (* (Math|Vector|Normalize (Math|Vector|GetForwardVector (Math|Random|RandomRotatorfromStream (Variables|Default|GetStream)))) (Math|Random|RandomFloatinRangefromStream (Variables|Default|GetStream) (Variables|Default|GetMinRadius) (Variables|Default|GetMax%20Radius)))) "0,0,-50000.000000") :ObjectTypes (Utilities|Array|MakeArray "ObjectTypeQuery1") :bTraceComplex false :DrawDebugType "None" :bIgnoreSelf false :TraceColor "(R=1.000000,G=0.000000,B=0.000000,A=1.000000)" :TraceHitColor "(R=0.000000,G=1.000000,B=0.000000,A=1.000000)" :DrawTime 5.0))
          (bind _array_element_2 (Utilities|Array|ForEachLoopwithBreak _outhits_2)
            (:LoopBody
              (if (== (.HitActor (Collision|BreakHitResult _array_element_2)) self)
                (LocalVariables|SetDistanceToSelf (.Distance (Collision|BreakHitResult _array_element_2))))
              (LocalVariables|SetDistanceToSelf 9999999999999.0))
            (:Completed
              _))
          (bind _array_element_3 (Utilities|Array|ForEachLoopwithBreak _outhits_2)
            (:LoopBody
              (Utilities|Casting|CastToLandscapeProxy (.HitActor (Collision|BreakHitResult _array_element_3)))
              (Utilities|FlowControl|Sequence
                (:then_0
                  _)
                (:then_1
                  _))
              (if (> (LocalVariables|GetDistanceToSelf) (.Distance (Collision|BreakHitResult _array_element_3)))
                (Components|InstancedStaticMesh|AddInstance (Variables|Components|GetInstancedStaticMesh) (Math|Transform|MakeTransform (- (.Location (Collision|BreakHitResult _array_element_3)) (Transformation|GetActorLocation)) (Math|Rotator|CombineRotators (Math|Rotator|MakeRotationfromAxes "1.000000, 0, 0" "0,1.000000, 0" (.Normal (Collision|BreakHitResult _array_element_3))) (Math|Rotator|MakeRotfromX (Math|Vector|RotateVectorAroundAxis "1.000000, 0, 0" (* (Math|Float|Fraction (Math|Random|RandomFloatinRangefromStream (Variables|Default|GetStream) 343.0 65464.0)) 3.6e+02) "0, 0,1.000000"))) (Math|Vector|MakeVector (Math|Random|RandomFloatinRangefromStream (Variables|Default|GetStream) (Variables|Default|GetMin_Scale_2ndScatter) (Variables|Default|GetMax_Scale_2ndScatter)) (Math|Random|RandomFloatinRangefromStream (Variables|Default|GetStream) (Variables|Default|GetMin_Scale_2ndScatter) (Variables|Default|GetMax_Scale_2ndScatter)) (Math|Random|RandomFloatinRangefromStream (Variables|Default|GetStream) (Variables|Default|GetMin_Scale_2ndScatter) (Variables|Default|GetMax_Scale_2ndScatter)))) false)))))))))

(fn ConstructionScript
  (bind _landscapetexturepatch (Variables|Default|GetLandscapeTexturePatch))
  (LandscapePatch|SetIsEnabled _landscapetexturepatch true)
  (LandscapePatch|RequestLandscapeUpdate _landscapetexturepatch false true)
  (CallFunction|AssignRiver)
  (bind (_newspline1 _waterbodyspline1 _baselocation1 _basedistance1) (CallFunction|CopySplineSection))
  (CallFunction|UpdateHeight)
  (CallFunction|SpawnLargeBankMeshes _newspline1 _waterbodyspline1 _baselocation1 _basedistance1)
  (CallFunction|SpawnRocksScatter _newspline1 _waterbodyspline1 _baselocation1 _basedistance1))
