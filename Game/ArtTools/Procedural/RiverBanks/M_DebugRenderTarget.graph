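# Unlit debug material for viewing a render target, kept with the procedural river bank art tools.
# It samples the Texture parameter and divides its RGB by the Range scalar, rescaling the stored values for display.
# Range defaults to one, so the texture values pass through unchanged unless it is overridden.
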Material /Game/ArtTools/Procedural/RiverBanks/M_DebugRenderTarget
graph {
 properties {pixelDepthOffsetMode=PDOM_Legacy, shadingModel=MSM_Unlit, usedShadingModels=Unlit}
 nodes [Divide(A=TextureSampleParameter2D() {parameterName=Texture}.RGB, B=ScalarParameter() {defaultValue=1, parameterName=Range})]
}
