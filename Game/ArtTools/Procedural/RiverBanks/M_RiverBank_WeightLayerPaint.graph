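# Post-process material in the procedural river bank tools, named for weight-layer painting, computing bank height offsets and a mask.
# It decodes RG-packed landscape height and derives spline surface height from SplineHeight and Z parameters.
# Randomly offset gradient noise drives height variation, capped near the spline when terrain lies below it, scaled separately otherwise.
# An Alpha texture times four-sided UV edge fades forms the mask; no material output connections appear.
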Material /Game/ArtTools/Procedural/RiverBanks/M_RiverBank_WeightLayerPaint
graph {
 properties {
  bAllowNegativeEmissiveColor=true
  blendableOutputAlpha=true
  materialDomain=MD_PostProcess
  pixelDepthOffsetMode=PDOM_Legacy
  usedShadingModels=Default Lit
 }
 nodes [
  Constant()
  n1 = TextureCoordinate()
  n2 = ScalarParameter() {parameterName=Random}
  n3 = Add(A=AppendVector(A=@n1, B=Constant()), B=AppendVector(A=AppendVector(A=@n2, B=@n2), B=@n2))
  n4 = ScalarParameter() {defaultValue=2, parameterName=WorldSize}
  Divide(A=@n3, B=@n4)
  n5 = ScalarParameter() {defaultValue=2, parameterName=Size}
  n6 = Noise!(World Position=Divide(A=Multiply(A=@n3, B=@n4), B=Multiply(B=@n5) {constA=1000}), FilterWidth) {bTurbulence=false, levelScale=1, levels=0, noiseFunction=NOISEFUNCTION_GradientTex, outputMin=-0.5, scale=5}
  n7 = TextureSampleParameter2D() {parameterName=LandscapeHeight}
  n8 = Multiply(A=Add(A=Multiply(A=@n7.R) {constB=256}, B=@n7.G)) {constB=255}
  n9 = Multiply(A=Divide(A=Subtract(A=@n8) {constB=32768}) {constB=128}, B=ScalarParameter() {defaultValue=100, parameterName=Landscape_ZScale})
  n10 = Add(A=Multiply(A=@n6, B=ScalarParameter() {defaultValue=1000, parameterName=HeightBelow}), B=@n9)
  n11 = Subtract(
   A=Subtract(A=ScalarParameter() {parameterName=Z_PositionCamera}, B=TextureSampleParameter2D() {parameterName=SplineHeight}.R)
   B=ScalarParameter() {parameterName=Z_PositionTerrain}
  )
  n12 = Add(A=ScalarParameter() {defaultValue=50, parameterName=MaxHeightAboveWater}, B=@n11)
  n13 = Constant() {r=1}
  LinearInterpolate(
   A=Subtract(A=Clamp(None=@n10, Max=@n12) {clampMode=CMODE_ClampMax}, B=@n9)
   B=Multiply(A=@n6, B=ScalarParameter() {defaultValue=100, parameterName=HeightAbove})
   Alpha=If(A=@n11, B=@n9, "A > B"=Constant(), "A == B"=@n13, "A < B"=@n13)
  )
  LinearInterpolate(Alpha=Saturate(None=@n6)) {constA=1, constB=-1}
  n14 = ComponentMask(None=@n1) {r=true}
  n15 = Divide(A=ScalarParameter() {defaultValue=2, parameterName=EdgeFade}, B=@n4)
  n16 = Saturate(None=Divide(A=@n14, B=@n15))
  n17 = Saturate(None=Divide(A=OneMinus(None=@n14), B=@n15))
  n18 = ComponentMask(None=@n1) {g=true}
  n19 = Saturate(None=Divide(A=@n18, B=@n15))
  n20 = Saturate(None=Divide(A=OneMinus(None=@n18), B=@n15))
  Multiply(A=TextureSampleParameter2D() {parameterName=Alpha}.R, B=Multiply(A=Multiply(A=@n16, B=@n17), B=Multiply(A=@n19, B=@n20)))
 ]
}
